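# Clear coat body material for the WurstLivery4 skin of the Wurst vehicle, set up for skeletal meshes and mesh deformers.
# The MF_Wurst_Body_WurstLivery4 function supplies the base material attributes, which pass through MF_BaseVehicle_Damage and then MF_VehicleDeRez_Transition.
# The clear coat bottom normal is split off before those layers and routed through a named reroute to the clear coat normal custom output.
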Material /VehicleCosmetics/Bodies/Wurst/Skins/WurstLivery4/MAT_Wurst_Body_WurstLivery4
graph {
 properties {
  bUseMaterialAttributes=true
  bUsedWithMeshDeformer=true
  bUsedWithSkeletalMesh=true
  pixelDepthOffsetMode=PDOM_Legacy
  shadingModel=MSM_ClearCoat
  usedShadingModels=Clear Coat
 }
 nodes [
  ClearCoatNormalCustomOutput!(None=NamedRerouteUsage())
  n1 = GetMaterialAttributes(None=MaterialFunctionCall() {materialFunction=&/MF_Wurst_Body_WurstLivery4}.Result) {attributeGetTypes [#1]}
  MaterialFunctionCall(In=MaterialFunctionCall(Material=@n1.MaterialAttributes) {materialFunction=&MF_BaseVehicle_Damage}.Result) {materialFunction=&MF_VehicleDeRez_Transition}
  NamedRerouteDeclaration(None=@n1.ClearCoatBottomNormal) {name=BodyBase_BottomNormal, nodeColor=<rgba 0 1 0.3672 1>}
 ]
}
